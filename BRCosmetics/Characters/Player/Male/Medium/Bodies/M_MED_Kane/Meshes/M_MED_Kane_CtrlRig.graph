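# Control Rig for the Kane character body that drives right-side shoulder helper bones; the left side has no logic.
# The execution chain constrains scapula_00_r to the clavicle with offset, then aims scapula_01_r using a helper bone and the pectoral and lat bones at the upper arm.
# A separate branch, not linked to Begin Execution, debug-draws a Bezier from four cv bones and fits the deltoid helper chain to it, stretched.
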ControlRigBlueprint /BRCosmetics/Characters/Player/Male/Medium/Bodies/M_MED_Kane/Meshes/M_MED_Kane_CtrlRig
graphs [
 RigVMFunctionLibrary {links [], nodes []}
 RigVMModel {nodes [
   n1 = &RigUnit_BeginExecution()
   /Script/ControlRig.RigUnit_TransformConstraint(
    BaseBone=None
    BaseTransform=(Rotation=(X=0,Y=0,Z=0,W=1),Translation=(X=0,Y=0,Z=0),Scale3D=(X=1,Y=1,Z=1))
    BaseTransformSpace=BaseSpace
    Targets=((Transform=(Rotation=(X=0,Y=0,Z=0,W=1),Translation=(X=0,Y=0,Z=0),Scale3D=(X=1,Y=1,Z=1)),Weight=1,bMaintainOffset=True,Filter=(TranslationFilter=(bX=True,bY=True,bZ=True),RotationFilter=(bX=True,bY=True,bZ=True),ScaleFilter=(bX=True,bY=True,bZ=True))))
    bUseInitialTransforms=True
    ExecutePin=@n1.ExecutePin
    Bone=/Script/ControlRig.RigUnit_BoneName(Bone=scapula_00_r) {is=struct RigVMUnitNode}.Bone
    Targets.0.Transform=&RigUnit_GetBoneTransform(Bone=&RigUnit_BoneName(Bone=clavicle_r).Bone).Transform
   ) {is=struct RigVMUnitNode}
   > /Script/ControlRig.RigUnit_AimBone(
    Bone=scapula_01_r
    DebugSettings=(bEnabled=true,Scale=25,WorldOffset=(Rotation=(X=0,Y=0,Z=0,W=1),Translation=(X=0,Y=0,Z=0),Scale3D=(X=1,Y=1,Z=1)))
    Primary=(Weight=1,Axis=(X=1,Y=0,Z=0),Target=(X=1,Y=0,Z=0),Kind=Location,Space="None")
    Secondary=(Weight=1,Axis=(X=0,Y=0,Z=1),Target=(X=1,Y=0,Z=0),Kind=Direction,Space="deltoidhelper_01_r")
    Weight=1
    bPropagateToChildren=True
    Secondary.Target=&RigUnit_GetBoneTransform(Bone=&RigUnit_BoneName(Bone=scapula_01_r_aim).Bone).Transform.Translation
   ) {is=struct RigVMUnitNode}
   n2 = &RigUnit_BoneName(Bone=pectoralis_01_r)
   > &RigUnit_AimBone(
    DebugSettings=(bEnabled=true,Scale=25,WorldOffset=(Rotation=(X=0,Y=0,Z=0,W=1),Translation=(X=0,Y=0,Z=0),Scale3D=(X=1,Y=1,Z=1)))
    Primary=(Weight=1,Axis=(X=1,Y=0,Z=0),Target=(X=1,Y=0,Z=0),Kind=Location,Space="None")
    Secondary=(Weight=1,Axis=(X=0,Y=0,Z=1),Target=(X=1,Y=0,Z=0),Kind=Direction,Space="deltoidhelper_01_r")
    Weight=1
    bPropagateToChildren=True
    Bone=@n2.Bone
    Primary.Target=&RigUnit_GetBoneTransform(Bone=&RigUnit_BoneName(Bone=upperarm_r).Bone).Transform.Translation
   )
   n3 = &RigUnit_BoneName(Bone=lattissimus_00_r)
   > &RigUnit_AimBone(
    DebugSettings=(bEnabled=true,Scale=25,WorldOffset=(Rotation=(X=0,Y=0,Z=0,W=1),Translation=(X=0,Y=0,Z=0),Scale3D=(X=1,Y=1,Z=1)))
    Primary=(Weight=1,Axis=(X=1,Y=0,Z=0),Target=(X=1,Y=0,Z=0),Kind=Location,Space="None")
    Secondary=(Weight=1,Axis=(X=0,Y=0,Z=1),Target=(X=1,Y=0,Z=0),Kind=Direction,Space="deltoidhelper_01_r")
    Weight=1
    bPropagateToChildren=True
    Bone=@n3.Bone
    Primary.Target=&RigUnit_GetBoneTransform(Bone=&RigUnit_BoneName(Bone=upperarm_r).Bone).Transform.Translation
   )
   n4 = /Script/RigVM.RigVMFunction_MathVectorMakeBezierFourPoint(
    Bezier=(A=(X=0,Y=0,Z=0),B=(X=0,Y=0,Z=0),C=(X=0,Y=0,Z=0),D=(X=0,Y=0,Z=0))
    Bezier.A=&RigUnit_GetBoneTransform(Bone=cv01).Transform.Translation
    Bezier.B=&RigUnit_GetBoneTransform(Bone=cv02).Transform.Translation
    Bezier.C=&RigUnit_GetBoneTransform(Bone=cv03).Transform.Translation
    Bezier.D=&RigUnit_GetBoneTransform(Bone=cv04).Transform.Translation
   ) {is=struct RigVMUnitNode}
   /Script/ControlRig.RigUnit_DebugBezier(
    Color=(R=1,G=0,B=0,A=1)
    DebugDrawSettings=(DepthPriority=SDPG_Foreground,Lifetime=-1)
    Detail=16
    ExecutePin=()
    MaximumU=1
    MinimumU=0
    Space=None
    Thickness=0
    WorldOffset=(Rotation=(X=0,Y=0,Z=0,W=1),Translation=(X=0,Y=0,Z=0),Scale3D=(X=1,Y=1,Z=1))
    bEnabled=True
    Bezier=@n4.Bezier
   ) {is=struct RigVMUnitNode}
   > /Script/ControlRig.RigUnit_FitChainToCurve(
    Alignment=Stretched
    DebugSettings=(bEnabled=true,Scale=1,CurveColor=(R=1,G=1,B=0,A=1),SegmentsColor=(R=1,G=0,B=0,A=1),WorldOffset=(Rotation=(X=0,Y=0,Z=0,W=1),Translation=(X=0,Y=0,Z=0),Scale3D=(X=1,Y=1,Z=1)))
    EndBone=deltoidhelper_03_r
    Maximum=1
    Minimum=0
    PrimaryAxis=(X=1,Y=0,Z=0)
    RotationEaseType=Linear
    Rotations=((Rotation=(X=0,Y=0,Z=0,W=1),Ratio=0))
    SamplingPrecision=12
    SecondaryAxis=(X=0,Y=0,Z=0)
    StartBone=deltoidhelper_01_r
    Weight=1
    bPropagateToChildren=True
    PoleVectorPosition=/Script/RigVM.RigVMFunction_MathVectorBezierFourPoint(T=0, Bezier=@n4.Bezier) {is=struct RigVMUnitNode}.Tangent
    Bezier=@n4.Bezier
   ) {is=struct RigVMUnitNode}
   &RigUnit_GetBoneTransform(Bone=@n2.Bone)
   &RigUnit_GetBoneTransform(Bone=@n3.Bone)
  ]}
]
hierarchy [
 root {loc=<0 0 0>, rot=<0 0 0 1>}
 pelvis {parent=root, loc=<0 -0.7071 93.1191>, rot=<0.0127 0.707 -0.0127 -0.707>}
 spine_01 {parent=pelvis, loc=<6.7494 -0.6388 0>, rot=<0 0 0.034 -0.9994>}
 spine_02 {parent=spine_01, loc=<7.2518 -0.035 0>, rot=<0 0 0.0238 -0.9997>}
 spine_03 {parent=spine_02, loc=<6.0104 0 0>, rot=<0 0 0.0149 -0.9999>}
 spine_04 {parent=spine_03, loc=<8.1024 0 0>, rot=<0 0 -0.0132 -0.9999>}
 spine_05 {parent=spine_04, loc=<10.0146 0 0>, rot=<0 0 -0.0015 -1>}
 clavicle_l {parent=spine_05, loc=<4.3575 -0.1466 -2.3223>, rot=<0.0465 -0.7289 -0.009 -0.6829>}
 upperarm_l {parent=clavicle_l, loc=<15.5111 0 0>, rot=<0.0226 -0.3234 -0.0088 -0.9459>}
 lowerarm_l {parent=upperarm_l, loc=<26.6514 0 0>, rot=<0.0391 -0.0099 -0.2453 -0.9686>}
 hand_l {parent=lowerarm_l, loc=<22.3491 0 0>, rot=<-0.7081 -0.0507 0.0975 -0.6975>}
 index_metacarpal_l {parent=hand_l, loc=<1.6535 0.0183 -2.96>, rot=<0.0158 -0.0124 -0.0173 -0.9996>}
 index_01_l {parent=index_metacarpal_l, loc=<8.3008 0.0357 -0.0524>, rot=<-0.0044 0.0485 0.0924 -0.9945>}
 index_02_l {parent=index_01_l, loc=<3.3587 -0.3004 0.0452>, rot=<-0.002 0.0027 0.3225 -0.9466>}
 index_03_l {parent=index_02_l, loc=<2.5578 -0.123 -0.0252>, rot=<-0.002 0.0067 0.2558 -0.9667>}
 middle_metacarpal_l {parent=hand_l, loc=<2.1856 -0.1725 -0.5984>, rot=<0.006 0.0168 -0.0329 -0.9993>}
 middle_01_l {parent=middle_metacarpal_l, loc=<7.3275 -0.2447 0.0129>, rot=<-0.0208 0.0042 0.0858 -0.9961>}
 middle_02_l {parent=middle_01_l, loc=<4.0123 0.1401 0.1175>, rot=<-0.0218 0.0258 0.3512 -0.9357>}
 middle_03_l {parent=middle_02_l, loc=<2.6077 0.0534 0.1373>, rot=<0.0082 0.0188 0.1708 -0.9851>}
 pinky_metacarpal_l {parent=hand_l, loc=<1.7616 -0.9753 3.3814>, rot=<-0.0397 0.0814 -0.069 -0.9935>}
 pinky_01_l {parent=pinky_metacarpal_l, loc=<7.1709 -0.4831 -0.1094>, rot=<-0.0654 -0.0035 0.2242 -0.9723>}
 pinky_02_l {parent=pinky_01_l, loc=<2.8274 0.0893 -0.1043>, rot=<0.0037 0.0011 0.2854 -0.9584>}
 pinky_03_l {parent=pinky_02_l, loc=<2.2046 0.0307 -0.0078>, rot=<-0.0002 0.0098 0.1445 -0.9895>}
 ring_metacarpal_l {parent=hand_l, loc=<2.2074 -0.3625 1.3638>, rot=<-0.0284 0.0722 -0.0458 -0.9959>}
 ring_01_l {parent=ring_metacarpal_l, loc=<7.0295 -0.6197 -0.1446>, rot=<-0.0489 -0.0252 0.1892 -0.9804>}
 ring_02_l {parent=ring_01_l, loc=<3.6174 0.4646 -0.1027>, rot=<0 0 0.2806 -0.9598>}
 ring_03_l {parent=ring_02_l, loc=<2.4152 0.1141 0.0199>, rot=<0 0 0.1189 -0.9929>}
 thumb_01_l {parent=hand_l, loc=<1.3118 -0.3191 -2.459>, rot=<0.6546 -0.4042 0.0664 -0.6354>}
 thumb_02_l {parent=thumb_01_l, loc=<5.9336 -0.0448 0.0401>, rot=<-0.0037 0.0213 0.298 -0.9543>}
 thumb_03_l {parent=thumb_02_l, loc=<2.2223 0.3682 -0.0907>, rot=<0.0118 -0.013 0.0066 -0.9998>}
 lowerarm_twist_01_l {parent=lowerarm_l, loc=<17.8549 0 0>, rot=<0 0 0 1>}
 lowerarm_twist_02_l {parent=lowerarm_l, loc=<9.6199 0 0>, rot=<0 0 0 1>}
 upperarm_twist_01_l {parent=upperarm_l, loc=<0.0041 0 0>, rot=<0 0 0 -1>}
 upperarm_twist_02_l {parent=upperarm_l, loc=<13.7941 0 0>, rot=<0 0 0 -1>}
 scapula_base_l {parent=clavicle_l, loc=<15.5111 0 -0.0001>, rot=<0.7302 0.0172 -0.6828 0.0184>}
 scapula_00_l {parent=scapula_base_l, loc=<-0.886 8.2456 12.7178>, rot=<0.768 -0.0679 -0.0021 -0.6368>}
 scapula_01_l {parent=scapula_00_l, loc=<9.9994 0.0001 0>, rot=<0 0 0 -1>}
 clavicle_r {parent=spine_05, loc=<4.3575 -0.1466 2.3223>, rot=<0.7289 0.0465 -0.6829 0.009>}
 upperarm_r {parent=clavicle_r, loc=<-15.5111 0 0>, rot=<0.0226 -0.3234 -0.0088 -0.9459>}
 lowerarm_r {parent=upperarm_r, loc=<-26.6514 0 0>, rot=<0.0391 -0.0099 -0.2453 -0.9686>}
 hand_r {parent=lowerarm_r, loc=<-22.3492 0 0>, rot=<-0.7081 -0.0507 0.0975 -0.6975>}
 index_metacarpal_r {parent=hand_r, loc=<-1.6534 -0.0181 2.9601>, rot=<0.0158 -0.0124 -0.0173 -0.9996>}
 index_01_r {parent=index_metacarpal_r, loc=<-8.3008 -0.0357 0.0524>, rot=<-0.0044 0.0485 0.0924 -0.9945>}
 index_02_r {parent=index_01_r, loc=<-3.3589 0.3003 -0.0453>, rot=<-0.002 0.0027 0.3225 -0.9466>}
 index_03_r {parent=index_02_r, loc=<-2.5578 0.1231 0.0251>, rot=<-0.002 0.0067 0.2558 -0.9667>}
 middle_metacarpal_r {parent=hand_r, loc=<-2.1856 0.1726 0.5984>, rot=<0.006 0.0168 -0.0329 -0.9993>}
 middle_01_r {parent=middle_metacarpal_r, loc=<-7.3273 0.2447 -0.0129>, rot=<-0.0208 0.0042 0.0858 -0.9961>}
 middle_02_r {parent=middle_01_r, loc=<-4.0142 -0.1411 -0.07>, rot=<-0.0217 0.0259 0.3512 -0.9357>}
 middle_03_r {parent=middle_02_r, loc=<-2.6074 -0.0534 -0.1312>, rot=<0.0083 0.0188 0.1708 -0.9851>}
 pinky_metacarpal_r {parent=hand_r, loc=<-1.7617 0.9753 -3.3814>, rot=<-0.0397 0.0814 -0.069 -0.9935>}
 pinky_01_r {parent=pinky_metacarpal_r, loc=<-7.1707 0.4832 0.1094>, rot=<-0.0654 -0.0035 0.2242 -0.9723>}
 pinky_02_r {parent=pinky_01_r, loc=<-2.8274 -0.0892 0.1043>, rot=<0.0037 0.0011 0.2854 -0.9584>}
 pinky_03_r {parent=pinky_02_r, loc=<-2.2045 -0.031 0.0078>, rot=<-0.0002 0.0098 0.1445 -0.9895>}
 ring_metacarpal_r {parent=hand_r, loc=<-2.2073 0.3627 -1.3638>, rot=<-0.0284 0.0722 -0.0458 -0.9959>}
 ring_01_r {parent=ring_metacarpal_r, loc=<-7.0297 0.6197 0.1446>, rot=<-0.0489 -0.0252 0.1892 -0.9804>}
 ring_02_r {parent=ring_01_r, loc=<-3.6172 -0.4648 0.1027>, rot=<0 0 0.2806 -0.9598>}
 ring_03_r {parent=ring_02_r, loc=<-2.4152 -0.1141 -0.0199>, rot=<0 0 0.1189 -0.9929>}
 thumb_01_r {parent=hand_r, loc=<-1.3119 0.3192 2.459>, rot=<0.6546 -0.4042 0.0664 -0.6354>}
 thumb_02_r {parent=thumb_01_r, loc=<-5.9335 0.0447 -0.0401>, rot=<-0.0037 0.0213 0.298 -0.9543>}
 thumb_03_r {parent=thumb_02_r, loc=<-2.2223 -0.3682 0.0907>, rot=<0.0118 -0.013 0.0066 -0.9998>}
 lowerarm_twist_01_r {parent=lowerarm_r, loc=<-17.855 0 0>, rot=<0 0 0 -1>}
 lowerarm_twist_02_r {parent=lowerarm_r, loc=<-9.6199 0 0>, rot=<0 0 0 -1>}
 cv04 {parent=lowerarm_r, loc=<0 0 0>, rot=<-0.2259 0.2795 0.8018 -0.4774>}
 upperarm_twist_01_r {parent=upperarm_r, loc=<-0.0041 0 0>, rot=<0 0 0 1>}
 upperarm_twist_02_r {parent=upperarm_r, loc=<-13.7942 0 0>, rot=<0 0 0 1>}
 cv02 {parent=upperarm_r, loc=<6.8781 -0.5172 0.776>, rot=<-0.2608 0.2419 0.6509 -0.6707>}
 cv03 {parent=upperarm_r, loc=<-4.8202 0.6183 -2.9449>, rot=<-0.2608 0.2419 0.6509 -0.6707>}
 deltoidhelper_01_r {parent=clavicle_r, loc=<-0.1181 -0.006 -0.0079>, rot=<-0.0002 -0.9946 -0.0017 -0.1038>}
 deltoidhelper_02_r {parent=deltoidhelper_01_r, loc=<12.7992 0.006 0.0166>, rot=<0 -0.4252 0 -0.9051>}
 deltoidhelper_03_r {parent=deltoidhelper_02_r, loc=<30.5575 0 0>, rot=<0.0006 0 1 0>}
 scapula_base_r {parent=clavicle_r, loc=<-15.5111 0 0>, rot=<0.6828 0.0184 0.7302 -0.0172>}
 scapula_00_r {parent=scapula_base_r, loc=<-0.8862 -8.2457 12.7178>, rot=<-0.768 -0.0679 0.0021 -0.6368>}
 scapula_01_r {parent=scapula_00_r, loc=<10 0 0>, rot=<0 0 0 -1>}
 neck_01 {parent=spine_05, loc=<11.3138 0.1529 0>, rot=<0 0 -0.2003 -0.9797>}
 neck_02 {parent=neck_01, loc=<4.5627 0 0>, rot=<0 0 0.0504 -0.9987>}
 head {parent=neck_02, loc=<5.3914 0 0>, rot=<0 0 0.115 -0.9934>}
 dyn_cable_tp_01_l_MSTR {parent=head, loc=<2.8137 3.0838 -10.7863>, rot=<0.0012 -0.9904 -0.1379 0.0093>}
 dyn_cable_tp_01_l {parent=dyn_cable_tp_01_l_MSTR, loc=<0.2 0 0>, rot=<0 0 0 1>}
 dyn_cable_tp_02_l {parent=dyn_cable_tp_01_l, loc=<5.8615 0 0>, rot=<0 0 0 1>}
 dyn_cable_md_01_l_MSTR {parent=spine_05, loc=<4.6357 7.9412 -3.5485>, rot=<-0.3865 -0.4277 -0.2949 -0.7621>}
 dyn_cable_md_01_l {parent=dyn_cable_md_01_l_MSTR, loc=<0.2 0 0>, rot=<0 0 0 1>}
 dyn_cable_md_02_l {parent=dyn_cable_md_01_l, loc=<3.2798 0 0>, rot=<0 0 0 1>}
 walkietalkie_01_l {parent=spine_05, loc=<5.648 7.2785 -9.1846>, rot=<0.1273 0.0765 0.9428 0.2985>}
 dyn_cable_bt_01_l {parent=walkietalkie_01_l, loc=<6.0031 -1.8084 -0.0559>, rot=<-0.7852 -0.1932 -0.1207 -0.5759>}
 dyn_cable_bt_02_l {parent=dyn_cable_bt_01_l, loc=<11.6328 0 0>, rot=<0 0 0 -1>}
 cv01 {parent=spine_05, loc=<4.3575 -0.1466 2.4408>, rot=<-0.4795 -0.5196 0.5196 0.4795>}
 pectoralis_01_r {parent=spine_04, loc=<6.1949 10.4373 4.2121>, rot=<0.4277 0.1804 0.5285 -0.7108>}
 pectoralis_02_r {parent=pectoralis_01_r, loc=<14.5069 0 0>, rot=<0 0 0 1>}
 lattissimus_00_r {parent=spine_04, loc=<-0.2412 -6.5493 8.2532>, rot=<-0.6488 0.0934 -0.32 -0.6841>}
 lattissimus_01_r {parent=lattissimus_00_r, loc=<16.1021 0 0>, rot=<0 0 0 1>}
 dyn_strap_01_r_MSTR {parent=spine_04, loc=<4.6194 0.0971 0.4175>, rot=<0.52 0.0333 0.8534 0.0121>}
 dyn_strap_01_r {parent=dyn_strap_01_r_MSTR, loc=<0.2 0 0>, rot=<0 0 0 1>}
 dyn_strap_02_r {parent=dyn_strap_01_r, loc=<13.7527 0 0>, rot=<0 0 0 1>}
 scapula_01_r_aim {parent=spine_04, loc=<-0.0983 -8.8521 6.4181>, rot=<0.0285 -0.6361 0.7702 0.0359>}
 lattissimus_00_l {parent=spine_04, loc=<-0.2415 -6.5493 -8.2532>, rot=<-0.6841 -0.32 0.0934 -0.6488>}
 lattissimus_01_l {parent=lattissimus_00_l, loc=<16.1025 0.0003 0.0001>, rot=<0 0 0 1>}
 thigh_l {parent=pelvis, loc=<-2.5693 -0.0109 -7.7501>, rot=<0.0012 0.0572 0.0155 -0.9982>}
 calf_l {parent=thigh_l, loc=<-39.0484 0 0>, rot=<0.0006 -0.0161 -0.0391 -0.9991>}
 calf_twist_01_l {parent=calf_l, loc=<-35.8131 -0.0967 0.0397>, rot=<0.0237 0.0095 -0.0035 -0.9997>}
 calf_twist_02_l {parent=calf_l, loc=<-18.8064 -0.0967 0.0397>, rot=<0.0237 0.0095 -0.0035 -0.9997>}
 foot_l {parent=calf_l, loc=<-42.561 0.1077 0.4292>, rot=<0.0418 -0.0455 0.0343 -0.9975>}
 ball_l {parent=foot_l, loc=<-6.1599 13.8264 0.6806>, rot=<0.0001 -0.0001 -0.7063 -0.7079>}
 thigh_twist_01_l {parent=thigh_l, loc=<-21.6303 0 0>, rot=<0.0013 0 0 -1>}
 thigh_r {parent=pelvis, loc=<-2.5691 -0.0109 7.7501>, rot=<0.0572 -0.0012 0.9982 0.0155>}
 calf_r {parent=thigh_r, loc=<39.0487 0 0>, rot=<0.0006 -0.0161 -0.0391 -0.9991>}
 calf_twist_01_r {parent=calf_r, loc=<35.8132 0.0967 -0.0397>, rot=<0.0237 0.0095 -0.0035 -0.9997>}
 calf_twist_02_r {parent=calf_r, loc=<18.8064 0.0967 -0.0397>, rot=<0.0237 0.0095 -0.0035 -0.9997>}
 foot_r {parent=calf_r, loc=<42.5611 -0.1077 -0.4292>, rot=<0.0418 -0.0455 0.0343 -0.9975>}
 ball_r {parent=foot_r, loc=<6.1598 -13.8264 -0.6806>, rot=<0.0001 -0.0001 -0.7063 -0.7079>}
 thigh_twist_01_r {parent=thigh_r, loc=<21.6303 0 0>, rot=<0.0013 0 0 -1>}
 glut_01_r {parent=pelvis, loc=<-2.5625 -0.7919 7.746>, rot=<0.051 -0.0259 0.9062 -0.4189>}
 glut_01_l {parent=pelvis, loc=<-2.5626 -0.7919 -7.746>, rot=<-0.051 0.0259 0.9062 -0.4189>}
]
