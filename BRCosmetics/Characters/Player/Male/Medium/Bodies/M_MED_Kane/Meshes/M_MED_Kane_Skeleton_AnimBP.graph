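# Post-process style AnimBP for the M_MED_Kane player body, layering secondary motion and corrective deformation onto the parent mesh's pose.
# AnimGraph copies that pose, then runs a Control Rig, right-strap AnimDynamics, and additive glute and upper-arm twist adjustments.
# EventGraph derives correction alphas from thigh and arm deltas against ref pose, overriding shoulder inflation while skydiving.
# That setter chain is unconnected to the update event, and arm-swing values go unused.
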AnimBlueprint /BRCosmetics/Characters/Player/Male/Medium/Bodies/M_MED_Kane/Meshes/M_MED_Kane_Skeleton_AnimBP
target_skeleton=&/M_MED_Kane_Skeleton
graphs [
 AnimGraph {nodes [
   n1 = Copy Pose From Mesh() {bUseAttachedParent=true}
   n2 = Control Rig(Source=Component To Local(ComponentPose=Local To Component(LocalPose=@n1))) {
    alphaBoolBlend {blendOption=Linear}
    controlRigAssetReference {blueprintRigClass=&/M_MED_Kane_CtrlRig.M_MED_Kane_CtrlRig_C}
   }
   n3 = Anim Dynamics - Bone: dyn_strap_01_r(ComponentPose=Local To Component(LocalPose=@n2), Alpha=0.15, ExternalForce="0,0,0") {
    angularDampingOverride=1
    angularSpringConstant=200
    bAngularSpring=true
    bGravityOverrideInSimSpace=true
    bLinearSpring=true
    bOverrideAngularDamping=true
    gravityScale=0.35
    linearDampingOverride=0.5
    linearSpringConstant=150
    simulationSpace=RootRelative
    alphaBoolBlend {blendOption=Linear}
    boundBone {boneName=dyn_strap_01_r}
    componentAppliedLinearAccClamp=<2000 2000 3000>
    componentLinearAccScale=<1.5 4 1>
    componentLinearVelScale=<0.5 4 1>
    physicsBodyDefinitions [{
      sphereCollisionRadius=0
      boundBone {boneName=dyn_strap_01_r}
      boxExtents=<15 15 15>
      constraintSetup {angularLimitsMax=<5 10 10>, angularLimitsMin=<-5 -45 -10>, angularTarget=<1 0 0>, linearAxesMax=<2 2 2>, linearAxesMin=<-2 -2 -2>}
      localJointOffset=<8 0 0>
     }]
   }
   n4 = "Transform (Modify) Bone - Bone: glut_01_l"(
    ComponentPose=@n3
    Alpha=Get thighSwingDeltaCrouch_l() -> (thighSwingDeltaCrouch_l:double)
    Translation="0,0,0"
    Rotation="0.000,0.000,-10"
    Scale="1,1,1"
   ) {
    rotationMode=BMM_Additive
    scaleMode=BMM_Additive
    translationMode=BMM_Additive
    alphaBoolBlend {blendOption=Linear}
    boneToModify {boneName=glut_01_l}
    rotation {roll=-10}
   }
   n5 = "Transform (Modify) Bone - Bone: glut_01_r"(
    ComponentPose=@n4
    Alpha=Get thighSwingDeltaCrouch_r() -> (thighSwingDeltaCrouch_r:double)
    Translation="0,-1,-2"
    Rotation="0.000,0.000,-2.5"
    Scale="1,1,1"
   ) {
    rotationMode=BMM_Additive
    scaleMode=BMM_Additive
    translationMode=BMM_Additive
    alphaBoolBlend {blendOption=Linear}
    boneToModify {boneName=glut_01_r}
    rotation {roll=-2.5}
    translation {y=-1, z=-2}
   }
   n6 = "Transform (Modify) Bone - Bone: upperarm_twist_01_r"(ComponentPose=@n5, Alpha=Get shoulderInflate_r() -> (shoulderInflate_r:double), Translation="0,0,0", Rotation="0,0,0", Scale="1.1,1.1,1.1") {scaleMode=BMM_Additive, alphaBoolBlend {blendOption=Linear}, boneToModify {boneName=upperarm_twist_01_r}, scale=<1.1 1.1 1.1>}
   n7 = "Transform (Modify) Bone - Bone: upperarm_twist_01_l"(ComponentPose=@n6, Alpha=Get shoulderInflate_l() -> (shoulderInflate_l:double), Translation="0,0,0", Rotation="0,0,0", Scale="1.1,1.1,1.1") {scaleMode=BMM_Additive, alphaBoolBlend {blendOption=Linear}, boneToModify {boneName=upperarm_twist_01_l}, scale=<1.1 1.1 1.1>}
   Output Pose(Result=Component To Local(ComponentPose=@n7)) {layerGroup=None}
  ]}
 EventGraph {nodes [
   n1 = GetOwnerSkeletalMeshComponent()
   n2 = GetDeltaTransformFromRefPose(self=@n1, BoneName=upperarm_r, BaseName=clavicle_r)
   n3 = "Absolute (Float)"(A=BreakRotator(InRot=BreakTransform(InTransform=@n2).Rotation).Yaw)
   Set armSwingDelta_r(armSwingDelta_r:double=MapRangeClamped(Value=@n3, InRangeA=5, InRangeB=70)) -> (Output_Get:double)
   n4 = GetDeltaTransformFromRefPose(self=@n1, BoneName=upperarm_l, BaseName=clavicle_l)
   n5 = "Absolute (Float)"(A=BreakRotator(InRot=BreakTransform(InTransform=@n4).Rotation).Yaw)
   > Set armSwingDelta_l(armSwingDelta_l:double=MapRangeClamped(Value=@n5, InRangeA=5, InRangeB=70)) -> (Output_Get:double)
   n6 = Get bIsSkyDiving()
   > Set shoulderInflate_r(shoulderInflate_r:double=SelectFloat(A=0.5, B=MapRangeClamped(Value=@n3, InRangeA=90, InRangeB=130, OutRangeB=0.75), bPickA=@n6)) -> (Output_Get:double)
   > Set shoulderInflate_l(shoulderInflate_l:double=SelectFloat(A=0.5, B=MapRangeClamped(Value=@n5, InRangeA=10, InRangeB=80), bPickA=@n6)) -> (Output_Get:double)
   n7 = BreakRotator(InRot=BreakTransform(InTransform=GetDeltaTransformFromRefPose(self=@n1, BoneName=thigh_l, BaseName=pelvis)).Rotation)
   > Set thighSwingDeltaCrouch_l(thighSwingDeltaCrouch_l:double=MapRangeClamped(Value="Absolute (Float)"(A=@n7.Yaw), InRangeA=50, InRangeB=110)) -> (Output_Get:double)
   n8 = BreakRotator(InRot=BreakTransform(InTransform=GetDeltaTransformFromRefPose(self=@n1, BoneName=thigh_r, BaseName=pelvis)).Rotation)
   > Set thighSwingDeltaCrouch_r(thighSwingDeltaCrouch_r:double=MapRangeClamped(Value="Absolute (Float)"(A=@n8.Yaw), InRangeA=50, InRangeB=110)) -> (Output_Get:double)
   TryGetPawnOwner()
   Event BlueprintUpdateAnimation()
  ]}
]
variables [
 armSwingDelta_l {type=number}
 armSwingDelta_r {type=number}
 shoulderInflate_l {type=number}
 shoulderInflate_r {type=number}
 thighSwingDeltaCrouch_l {type=number}
 thighSwingDeltaCrouch_r {type=number}
]
