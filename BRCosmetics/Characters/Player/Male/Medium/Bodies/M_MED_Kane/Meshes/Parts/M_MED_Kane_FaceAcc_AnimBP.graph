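# Post-process-style anim Blueprint for Kane's face accessory mesh, adding walkie-talkie and cable sway plus a spinner.
# It copies the attached parent's pose, then chains angular-spring AnimDynamics on four bones and additively rotates spinner_01.
# While recently fired, emoting or skydiving, each update spins spinnerRot with increasing interpolation speed.
# Otherwise it eases back to zero at fixed speed; the computed spin-down speed is never applied.
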AnimBlueprint /BRCosmetics/Characters/Player/Male/Medium/Bodies/M_MED_Kane/Meshes/Parts/M_MED_Kane_FaceAcc_AnimBP
target_skeleton=&/M_MED_Kane_FaceAcc_Skeleton
graphs [
 AnimGraph {nodes [
   n1 = Anim Dynamics - Bone: walkietalkie_01_l(ComponentPose=Local To Component(LocalPose=Copy Pose From Mesh() {bUseAttachedParent=true}), Alpha=1, ExternalForce="0,0,0") {
    angularDampingOverride=1
    angularSpringConstant=1500
    bAngularSpring=true
    bGravityOverrideInSimSpace=true
    bOverrideAngularDamping=true
    gravityScale=0.35
    simulationSpace=RootRelative
    alphaBoolBlend {blendOption=Linear}
    boundBone {boneName=walkietalkie_01_l}
    physicsBodyDefinitions [{
      sphereCollisionRadius=0
      boundBone {boneName=walkietalkie_01_l}
      boxExtents=<10 20 20>
      constraintSetup {angularLimitsMax=<0 5 0>, angularLimitsMin=<0 -5 0>, angularTarget=<0 0 0>}
      localJointOffset=<10 0 0>
     }]
   }
   n2 = Anim Dynamics - Bone: dyn_cable_bt_01_l(ComponentPose=@n1, Alpha=1, ExternalForce="0,0,0") {
    angularDampingOverride=1
    angularSpringConstant=1500
    bAngularSpring=true
    bGravityOverrideInSimSpace=true
    bOverrideAngularDamping=true
    gravityScale=0.35
    simulationSpace=RootRelative
    alphaBoolBlend {blendOption=Linear}
    boundBone {boneName=dyn_cable_bt_01_l}
    physicsBodyDefinitions [{
      sphereCollisionRadius=0
      boundBone {boneName=dyn_cable_bt_01_l}
      boxExtents=<10 20 20>
      constraintSetup {angularLimitsMax=<0 0 0>, angularLimitsMin=<0 -40 -40>, angularTarget=<0 0 0>}
      localJointOffset=<10 0 0>
     }]
   }
   n3 = Anim Dynamics - Bone: dyn_cable_md_01_l(ComponentPose=@n2, Alpha=1, ExternalForce="0,0,0") {
    angularDampingOverride=1
    angularSpringConstant=1500
    bAngularSpring=true
    bGravityOverrideInSimSpace=true
    bOverrideAngularDamping=true
    gravityScale=0.35
    simulationSpace=RootRelative
    alphaBoolBlend {blendOption=Linear}
    boundBone {boneName=dyn_cable_md_01_l}
    physicsBodyDefinitions [{
      sphereCollisionRadius=0
      boundBone {boneName=dyn_cable_md_01_l}
      boxExtents=<10 20 20>
      constraintSetup {angularLimitsMax=<0 0 0>, angularLimitsMin=<0 0 -40>, angularTarget=<0 0 0>}
      localJointOffset=<10 0 0>
     }]
   }
   n4 = Anim Dynamics - Bone: dyn_cable_tp_01_l(ComponentPose=@n3, Alpha=1, ExternalForce="0,0,0") {
    angularDampingOverride=1
    angularSpringConstant=1500
    bAngularSpring=true
    bGravityOverrideInSimSpace=true
    bOverrideAngularDamping=true
    gravityScale=0.35
    simulationSpace=RootRelative
    alphaBoolBlend {blendOption=Linear}
    boundBone {boneName=dyn_cable_tp_01_l}
    physicsBodyDefinitions [{
      sphereCollisionRadius=0
      boundBone {boneName=dyn_cable_tp_01_l}
      boxExtents=<10 20 20>
      constraintSetup {angularLimitsMax=<0 0 0>, angularLimitsMin=<0 0 -40>, angularTarget=<0 0 0>}
      localJointOffset=<10 0 0>
     }]
   }
   n5 = "Transform (Modify) Bone - Bone: spinner_01"(ComponentPose=@n4, Alpha=1, Translation="0,0,0", Rotation=Get spinnerRot() -> (spinnerRot:Rotator), Scale="1,1,1") {rotationMode=BMM_Additive, rotationSpace=BCS_BoneSpace, alphaBoolBlend {blendOption=Linear}, boneToModify {boneName=spinner_01}}
   Output Pose(Result=Component To Local(ComponentPose=@n5)) {layerGroup=None}
  ]}
 EventGraph {nodes [
   n1 = Delay(Duration=0.1)
   n2 = Event BlueprintUpdateAnimation()
   n3 = Set spinnerRot(
    execute=@n16.then
    spinnerRot:Rotator=RInterpTo!(Current:Rotator=Get spinnerRot(), Target:Rotator="0, 0, 0", DeltaTime:float=@n2.DeltaTimeX, InterpSpeed:float=5) -> (ReturnValue:Rotator)
   ) -> (Output_Get:Rotator)
   n4 = "Absolute (Float)"(A=BreakRotator(InRot=@n3))
   Set interpSpeedUp(
    execute=[@n1.then, @n15.else]
    interpSpeedUp:double=SelectFloat(
     A=0
     B=Get interpSpeedUp() -> (interpSpeedUp:double)
     bPickA="float <= float"!(A:double=@n4, B:double=1) -> (ReturnValue:bool) {is=title K2Node_CallFunction}
    )
   ) -> (Output_Get:double)
   n5 = Increment Float!(" ":Exec, Value:"Float (double-precision) (by ref)"=Get interpSpeedUp()) -> ("  ":Exec, Result:double=~) {is=title K2Node_MacroInstance}
   > Set interpSpeedUp(interpSpeedUp=@n5)
   n6 = Get spinnerRot()
   Set spinnerRot(
    execute=@n14.then
    spinnerRot=RInterpTo(
     Current=Get spinnerRot()
     Target=Combine Rotators!(A:Rotator=@n6, B:Rotator="0, 0,30") -> (ReturnValue:Rotator)
     DeltaTime=@n2.DeltaTimeX
     InterpSpeed=Get interpSpeedUp()
    )
   )
   n7 = Set interpSpeedDown(
    execute=@n12.else
    interpSpeedDown:double=float + float(A=Get interpSpeedDown() -> (interpSpeedDown:double), B=0.5) {is=title K2Node_CommutativeAssociativeBinaryOperator}
   ) -> (Output_Get:double)
   n8 = MapRangeClamped(Value=@n7, InRangeA=1000, InRangeB=0, OutRangeB=25)
   n9 = "float > float"(A=@n4, B=3) {is=title K2Node_CallFunction}
   n10 = OR Boolean!(A=Get bRecentlyFired() -> (bRecentlyFired:bool), B=Get bIsPlayingEmote(), C:bool=Get bIsSkyDiving()) -> (ReturnValue)
   n11 = float + float(A=Get interpSpeedUp(), B=0.5)
   n12 = Branch(execute=@n2.then, Condition=@n10)
   > n13 = Increment Float(Value=Get numberOfTimesFired() -> (numberOfTimesFired:double)) -> (Result=~)
   > Set numberOfTimesFired(numberOfTimesFired:double=@n13) -> (Output_Get:double)
   > n14 = Set interpSpeedUp(interpSpeedUp=@n11)
   n15 = Branch(execute=@n3.then, Condition=@n9)
   n16 = Set interpSpeedDownInv(execute=@n7.then, interpSpeedDownInv:double=@n8) -> (Output_Get:double)
  ]}
]
variables [
 interpSpeedDown {type=number}
 interpSpeedDownInv {type=number}
 interpSpeedUp {type=number}
 numberOfTimesFired {type=number}
 spinnerRot {type=Rotator}
]
